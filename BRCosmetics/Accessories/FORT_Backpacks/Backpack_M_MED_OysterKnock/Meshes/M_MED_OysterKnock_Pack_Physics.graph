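# Physics asset for the OysterKnock backpack cosmetic, defining capsule bodies for a small figure with a head, body, lower body and three pairs of limbs.
# Upper limbs chain from the head, mid limbs from the main body and lower limbs from the lower body's dyn bone.
# Base-to-parent joints have free swing and twist, while each dyn-to-base joint has angular limits, tightest on the upper limbs and widest on the mid limbs.
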PhysicsAsset /BRCosmetics/Accessories/FORT_Backpacks/Backpack_M_MED_OysterKnock/Meshes/M_MED_OysterKnock_Pack_Physics
bodies {
 body_base [{shapeType=Capsule, length=0.8043, radius=3.5738, shapeName=body_base_capsule, center=<0 0.0139 -0.0362>, rotation=<rot -0.4299 -90 0>}]
 head_base [{shapeType=Capsule, length=0.505, radius=0.505, shapeName=head_base_capsule, rotation=<rot 0 -90 0>}]
 l_Mid_limb_base [{shapeType=Capsule, length=0.505, radius=0.505, shapeName=l_Mid_limb_base_capsule, rotation=<rot 90 0 0>}]
 l_Mid_limb_dyn [{shapeType=Capsule, length=1, radius=1, center=<1 0 0>, rotation=<rot 90 0 0>}]
 l_lower_limb_base [{shapeType=Capsule, length=0.505, radius=0.505, shapeName=l_lower_limb_base_capsule, rotation=<rot 0 -90 0>}]
 l_lower_limb_dyn [{shapeType=Capsule, length=1, radius=1, center=<1 0 0>, rotation=<rot 90 0 0>}]
 l_upper_limb_base [{shapeType=Capsule, length=0.505, radius=0.505, shapeName=l_upper_limb_base_capsule, rotation=<rot 0 -90 0>}]
 l_upper_limb_dyn [{shapeType=Capsule, length=1, radius=1, center=<1 0 0>, rotation=<rot 90 0 0>}]
 lower_body_base [{shapeType=Capsule, length=0.505, radius=0.505, shapeName=lower_body_base_capsule, rotation=<rot 0 -90 0>}]
 lower_body_dyn [{shapeType=Capsule, length=1, radius=1, center=<1 0 0>, rotation=<rot 90 0 0>}]
 r_Mid_limb_base [{shapeType=Capsule, length=0.505, radius=0.505, shapeName=r_Mid_limb_base_capsule, rotation=<rot 90 0 0>}]
 r_Mid_limb_dyn [{shapeType=Capsule, length=1, radius=1, center=<1 0 0>, rotation=<rot 90 0 0>}]
 r_lower_limb_base [{shapeType=Capsule, length=0.505, radius=0.505, shapeName=r_lower_limb_base_capsule, rotation=<rot 0 -90 0>}]
 r_lower_limb_dyn [{shapeType=Capsule, length=1, radius=1, center=<1 0 0>, rotation=<rot 90 0 0>}]
 r_upper_limb_base [{shapeType=Capsule, length=0.505, radius=0.505, shapeName=r_upper_limb_base_capsule, rotation=<rot 0 -90 0>}]
 r_upper_limb_dyn [{shapeType=Capsule, length=1, radius=1, center=<1 0 0>, rotation=<rot 90 0 0>}]
}
constraints [
 {
  bone1Name=l_Mid_limb_base
  bone2Name=body_base
  swing1LimitDegrees=45
  swing1Motion=Free
  swing2LimitDegrees=45
  swing2Motion=Free
  twistLimitDegrees=45
  twistMotion=Free
 }
 {
  bone1Name=l_lower_limb_base
  bone2Name=lower_body_dyn
  swing1LimitDegrees=45
  swing1Motion=Free
  swing2LimitDegrees=45
  swing2Motion=Free
  twistLimitDegrees=45
  twistMotion=Free
 }
 {
  bone1Name=r_Mid_limb_base
  bone2Name=body_base
  swing1LimitDegrees=45
  swing1Motion=Free
  swing2LimitDegrees=45
  swing2Motion=Free
  twistLimitDegrees=45
  twistMotion=Free
 }
 {
  bone1Name=r_lower_limb_base
  bone2Name=lower_body_dyn
  swing1LimitDegrees=45
  swing1Motion=Free
  swing2LimitDegrees=45
  swing2Motion=Free
  twistLimitDegrees=45
  twistMotion=Free
 }
 {
  bone1Name=l_upper_limb_base
  bone2Name=head_base
  swing1LimitDegrees=45
  swing1Motion=Free
  swing2LimitDegrees=45
  swing2Motion=Free
  twistLimitDegrees=45
  twistMotion=Free
 }
 {
  bone1Name=r_upper_limb_base
  bone2Name=head_base
  swing1LimitDegrees=45
  swing1Motion=Free
  swing2LimitDegrees=45
  swing2Motion=Free
  twistLimitDegrees=45
  twistMotion=Free
 }
 {
  bone1Name=head_base
  bone2Name=body_base
  swing1LimitDegrees=45
  swing1Motion=Free
  swing2LimitDegrees=45
  swing2Motion=Free
  twistLimitDegrees=45
  twistMotion=Free
 }
 {bone1Name=l_Mid_limb_dyn, bone2Name=l_Mid_limb_base, swing1LimitDegrees=20, swing2LimitDegrees=70, twistLimitDegrees=60}
 {bone1Name=l_lower_limb_dyn, bone2Name=l_lower_limb_base, swing1LimitDegrees=10, swing2LimitDegrees=45, twistLimitDegrees=45}
 {bone1Name=r_Mid_limb_dyn, bone2Name=r_Mid_limb_base, swing1LimitDegrees=20, swing2LimitDegrees=70, twistLimitDegrees=60}
 {bone1Name=r_lower_limb_dyn, bone2Name=r_lower_limb_base, swing1LimitDegrees=10, swing2LimitDegrees=45, twistLimitDegrees=45}
 {bone1Name=l_upper_limb_dyn, bone2Name=l_upper_limb_base, swing1LimitDegrees=5, swing2LimitDegrees=5, twistLimitDegrees=5}
 {bone1Name=r_upper_limb_dyn, bone2Name=r_upper_limb_base, swing1LimitDegrees=5, swing2LimitDegrees=5, twistLimitDegrees=5}
 {
  bone1Name=lower_body_base
  bone2Name=body_base
  swing1LimitDegrees=45
  swing1Motion=Free
  swing2LimitDegrees=45
  swing2Motion=Free
  twistLimitDegrees=45
  twistMotion=Free
 }
 {bone1Name=lower_body_dyn, bone2Name=lower_body_base, swing1LimitDegrees=15, swing2LimitDegrees=15, twistLimitDegrees=15}
]
