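# Anim Blueprint for the OysterKnock backpack, animating idle, combat and elimination reactions.
# A bool selects a front-end reactivity state machine or a gameplay one driven by aiming, firing, melee and elimination events.
# Pack offsets from owner spine twist and jumping/falling precede two RigidBody simulations: curve-weighted with overridden gravity, and front-end-only zero-gravity.
# The cached result gets a final pack offset; the limb and face bone-modifier chain never reaches output.
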AnimBlueprint /BRCosmetics/Accessories/FORT_Backpacks/Backpack_M_MED_OysterKnock/Meshes/M_MED_OysterKnock_Pack_AnimBP
target_skeleton=&/M_MED_OysterKnock_Pack_Skeleton
graphs [
 AnimGraph {nodes [
   n1 = "Transform (Modify) Bone - Bone: inner_face_dyn"!(ComponentPose, bAlphaBoolEnabled:bool=True, Rotation="0,0,0") -> (Pose) {
    alphaInputType=Bool
    rotationMode=BMM_Additive
    rotationSpace=BCS_BoneSpace
    alphaBoolBlend {blendInTime=0.2, blendOption=Sinusoidal, blendOutTime=0.45}
    boneToModify {boneName=inner_face_dyn}
   }
   n2 = "Transform (Modify) Bone - Bone: inner_face_dyn"(ComponentPose=@n1, bAlphaBoolEnabled=True, Rotation="0,0,0") {
    alphaInputType=Bool
    rotationMode=BMM_Additive
    rotationSpace=BCS_BoneSpace
    alphaBoolBlend {blendInTime=0.45, blendOption=Sinusoidal, blendOutTime=0.45}
    boneToModify {boneName=inner_face_dyn}
   }
   n3 = "Transform (Modify) Bone - Bone: r_Mid_limb_base"!(ComponentPose=@n2, bAlphaBoolEnabled:bool=True, Rotation="50,0,0") -> (Pose) {
    alphaInputType=Bool
    rotationMode=BMM_Additive
    rotationSpace=BCS_BoneSpace
    alphaBoolBlend {blendInTime=0.15, blendOption=Sinusoidal, blendOutTime=0.25}
    boneToModify {boneName=r_Mid_limb_base}
    rotation {pitch=50}
   }
   n4 = "Transform (Modify) Bone - Bone: l_Mid_limb_base"!(ComponentPose=@n3, bAlphaBoolEnabled:bool=True, Rotation="-50,0,0") -> (Pose) {
    alphaInputType=Bool
    rotationMode=BMM_Additive
    rotationSpace=BCS_BoneSpace
    alphaBoolBlend {blendInTime=0.15, blendOption=Sinusoidal, blendOutTime=0.25}
    boneToModify {boneName=l_Mid_limb_base}
    rotation {pitch=-50}
   }
   n5 = "Transform (Modify) Bone - Bone: r_lower_limb_base"!(ComponentPose=@n4, bAlphaBoolEnabled:bool=True, Rotation="-20,0,0") -> (Pose) {
    alphaInputType=Bool
    rotationMode=BMM_Additive
    rotationSpace=BCS_BoneSpace
    alphaBoolBlend {blendInTime=0.15, blendOption=Linear, blendOutTime=0.25}
    boneToModify {boneName=r_lower_limb_base}
    rotation {pitch=-20}
   }
   n6 = "Transform (Modify) Bone - Bone: l_lower_limb_base"!(ComponentPose=@n5, bAlphaBoolEnabled:bool=True, Rotation="-20,0,0") -> (Pose) {
    alphaInputType=Bool
    rotationMode=BMM_Additive
    rotationSpace=BCS_BoneSpace
    alphaBoolBlend {blendInTime=0.15, blendOption=Linear, blendOutTime=0.25}
    boneToModify {boneName=l_lower_limb_base}
    rotation {pitch=-20}
   }
   n7 = "Transform (Modify) Bone - Bone: l_upper_limb_base"!(ComponentPose=@n6, bAlphaBoolEnabled:bool=True, Rotation="-20,0,0") -> (Pose) {
    alphaInputType=Bool
    rotationMode=BMM_Additive
    rotationSpace=BCS_BoneSpace
    alphaBoolBlend {blendInTime=0.15, blendOption=Linear, blendOutTime=0.25}
    boneToModify {boneName=l_upper_limb_base}
    rotation {pitch=-20}
   }
   n8 = "Transform (Modify) Bone - Bone: r_upper_limb_base"!(ComponentPose=@n7, bAlphaBoolEnabled:bool=True, Rotation="-20,0,0") -> (Pose) {
    alphaInputType=Bool
    rotationMode=BMM_Additive
    rotationSpace=BCS_BoneSpace
    alphaBoolBlend {blendInTime=0.15, blendOption=Linear, blendOutTime=0.25}
    boneToModify {boneName=r_upper_limb_base}
    rotation {pitch=-20}
   }
   n9 = "Transform (Modify) Bone - Bone: r_Mid_limb_base"(ComponentPose=@n8, bAlphaBoolEnabled=True, Rotation="50,0,0") {
    alphaInputType=Bool
    rotationMode=BMM_Additive
    rotationSpace=BCS_BoneSpace
    alphaBoolBlend {blendInTime=0.15, blendOption=Sinusoidal, blendOutTime=0.25}
    boneToModify {boneName=r_Mid_limb_base}
    rotation {pitch=50}
   }
   n10 = "Transform (Modify) Bone - Bone: l_Mid_limb_base"(ComponentPose=@n9, bAlphaBoolEnabled=True, Rotation="-50,0,0") {
    alphaInputType=Bool
    rotationMode=BMM_Additive
    rotationSpace=BCS_BoneSpace
    alphaBoolBlend {blendInTime=0.15, blendOption=Sinusoidal, blendOutTime=0.25}
    boneToModify {boneName=l_Mid_limb_base}
    rotation {pitch=-50}
   }
   n11 = "Transform (Modify) Bone - Bone: r_lower_limb_base"(ComponentPose=@n10, bAlphaBoolEnabled=True, Rotation="-20,0,0") {
    alphaInputType=Bool
    rotationMode=BMM_Additive
    rotationSpace=BCS_BoneSpace
    alphaBoolBlend {blendInTime=0.15, blendOption=Linear, blendOutTime=0.25}
    boneToModify {boneName=r_lower_limb_base}
    rotation {pitch=-20}
   }
   n12 = "Transform (Modify) Bone - Bone: l_lower_limb_base"(ComponentPose=@n11, bAlphaBoolEnabled=True, Rotation="-20,0,0") {
    alphaInputType=Bool
    rotationMode=BMM_Additive
    rotationSpace=BCS_BoneSpace
    alphaBoolBlend {blendInTime=0.15, blendOption=Linear, blendOutTime=0.25}
    boneToModify {boneName=l_lower_limb_base}
    rotation {pitch=-20}
   }
   n13 = "Transform (Modify) Bone - Bone: l_upper_limb_base"(ComponentPose=@n12, bAlphaBoolEnabled=True, Rotation="-20,0,0") {
    alphaInputType=Bool
    rotationMode=BMM_Additive
    rotationSpace=BCS_BoneSpace
    alphaBoolBlend {blendInTime=0.15, blendOption=Linear, blendOutTime=0.25}
    boneToModify {boneName=l_upper_limb_base}
    rotation {pitch=-20}
   }
   "Transform (Modify) Bone - Bone: r_upper_limb_base"(ComponentPose=@n13, bAlphaBoolEnabled=True, Rotation="-20,0,0") {
    alphaInputType=Bool
    rotationMode=BMM_Additive
    rotationSpace=BCS_BoneSpace
    alphaBoolBlend {blendInTime=0.15, blendOption=Linear, blendOutTime=0.25}
    boneToModify {boneName=r_upper_limb_base}
    rotation {pitch=-20}
   }
   n14 = "Transform (Modify) Bone - Bone: pack"!(ComponentPose=Local To Component(LocalPose=Use cached pose 'AnimdynamicsCache'()), Alpha, Translation="0,0,4") -> (Pose) {translationMode=BMM_Additive, alphaBoolBlend {blendOption=Linear}, boneToModify {boneName=pack}, translation {z=4}}
   Output Pose(Result=Component To Local(ComponentPose=@n14)) {layerGroup=None}
   AnimGraphNode_StateMachine_1 = FE_Reactivity() {is=title AnimGraphNode_StateMachine}
   AnimGraphNode_StateMachine_2 = Backpack States() {is=title AnimGraphNode_StateMachine}
   n15 = Get bShouldDisplayFrontendReactivity() -> (bShouldDisplayFrontendReactivity:bool)
   n16 = "Transform (Modify) Bone - Bone: pack"!(
    ComponentPose=Local To Component(LocalPose=Blend Poses by bool(BlendPose_0=@AnimGraphNode_StateMachine_1, BlendPose_1=@AnimGraphNode_StateMachine_2, bActiveValue=@n15) {blendType=Linear})
    Alpha
    Translation=Get backpack_translation()
    Rotation:Rotator=Get backpack_rotate()
   ) -> (Pose) {translationMode=BMM_Additive, translationSpace=BCS_BoneSpace, alphaBoolBlend {blendOption=Linear}, boneToModify {boneName=pack}}
   n17 = "Transform (Modify) Bone - Bone: pack"!(ComponentPose=@n16, Alpha=Get JumpingFalling() -> (JumpingFalling:float), Translation="0,-1,0", Rotation="0,0,0", Scale:Vector) -> (Pose) {
    rotationMode=BMM_Additive
    rotationSpace=BCS_BoneSpace
    translationMode=BMM_Additive
    translationSpace=BCS_BoneSpace
    alphaBoolBlend {blendOption=Linear}
    alphaScaleBiasClamp {bInterpResult=true, interpSpeedDecreasing=5}
    boneToModify {boneName=pack}
    translation {y=-1}
   }
   n18 = RigidBody!(ComponentPose=@n17, AlphaCurveName:Name=DynamicAlpha, OverrideWorldGravity:Vector=Get Gravity Override(), ExternalForce) -> (Pose) {
    alphaInputType=Curve
    bOverrideWorldGravity=true
    overridePhysicsAsset=&/M_MED_OysterKnock_Pack_Physics
    simulationSpace=BaseBoneSpace
    alphaBoolBlend {blendOption=Linear}
    alphaScaleBiasClamp {interpSpeedDecreasing=3}
    baseBoneRef {boneName=root}
    componentAppliedLinearAccClamp=<2500 2500 2500>
    componentLinearAccScale=<0.25 0.25 0.25>
    componentLinearVelScale=<0.4 0.4 0.4>
    simSpaceSettings {
     maxAngularAcceleration=1000
     maxAngularVelocity=7
     maxLinearAcceleration=700
     maxLinearVelocity=500
     velocityScaleZ=0.6
     externalLinearDragV {x=0.2}
    }
   }
   n19 = RigidBody!(ComponentPose=@n18, bAlphaBoolEnabled:bool=Get bIsInFrontEnd(), OverrideWorldGravity="0,0,0", ExternalForce) -> (Pose) {
    alphaInputType=Bool
    bOverrideWorldGravity=true
    overridePhysicsAsset=&/M_MED_OysterKnock_Pack_Physicstest
    simulationSpace=BaseBoneSpace
    alphaBoolBlend {blendOption=Linear}
    alphaScaleBiasClamp {interpSpeedDecreasing=3}
    baseBoneRef {boneName=root}
    componentAppliedLinearAccClamp=<2500 2500 2500>
    componentLinearAccScale=<0.25 0.25 0.25>
    componentLinearVelScale=<0.4 0.4 0.4>
    simSpaceSettings {
     maxAngularAcceleration=1000
     maxAngularVelocity=7
     maxLinearAcceleration=700
     maxLinearVelocity=500
     velocityScaleZ=0.6
     externalLinearDragV {x=0.2}
    }
   }
   Save cached pose 'AnimdynamicsCache'(Pose=Component To Local(ComponentPose=@n19)) {cacheName=AnimdynamicsCache}
  ]}
 FE_Reactivity {
  entry=Idle
  path=AnimGraph.AnimGraphNode_StateMachine_1.FE_Reactivity
  ids=id
  nodes [
   AnimStateNode_0 = Idle()
   AnimStateNode_2 = FE_Elim() {is=title AnimStateNode}
   AnimStateTransitionNode_8 = AnimStateTransitionNode()
   AnimStateTransitionNode_9 = AnimStateTransitionNode() {bAutomaticRuleBasedOnSequencePlayerInState=true}
  ]
  links [
   {
    dst=Idle
    rule=AnimGraph.AnimGraphNode_StateMachine_1.FE_Reactivity.AnimStateTransitionNode_9.Transition
    src=FE_Elim
    via=AnimStateTransitionNode_9
   }
   {
    dst=FE_Elim
    rule=AnimGraph.AnimGraphNode_StateMachine_1.FE_Reactivity.AnimStateTransitionNode_8.Transition
    src=Idle
    via=AnimStateTransitionNode_8
   }
  ]
 }
 Idle {
  path=AnimGraph.AnimGraphNode_StateMachine_1.FE_Reactivity.AnimStateNode_0.Idle
  nodes [Output Animation Pose(Result="Sequence Evaluator (by time) 'OysterKnock_IdlePose_pack'"!(ExplicitTime:float=0) -> (Pose:Pose Link Structure) {
     is=title AnimGraphNode_SequenceEvaluator
     sequence=/BRCosmetics/Accessories/FORT_Backpacks/Backpack_M_MED_OysterKnock/Anim/OysterKnock_IdlePose_pack
    })]
 }
 FE_Elim {
  path=AnimGraph.AnimGraphNode_StateMachine_1.FE_Reactivity.AnimStateNode_2.FE_Elim
  nodes [Output Animation Pose(Result=Sequence Player 'OysterKnock_FE_Reactive_pack'() {bLoopAnimation=false, sequence=/BRCosmetics/Accessories/FORT_Backpacks/Backpack_M_MED_OysterKnock/Anim/OysterKnock_FE_Reactive_pack})]
 }
 Transition {
  path=AnimGraph.AnimGraphNode_StateMachine_1.FE_Reactivity.AnimStateTransitionNode_8.Transition
  nodes [Result(bCanEnterTransition=Get bEnemyPlayerEliminated() -> (bEnemyPlayerEliminated:bool))]
 }
 Transition {path=AnimGraph.AnimGraphNode_StateMachine_1.FE_Reactivity.AnimStateTransitionNode_9.Transition, nodes [Result()]}
 Backpack States {
  entry=Idle_pose
  path=AnimGraph.AnimGraphNode_StateMachine_2.Backpack States
  ids=id
  nodes [
   AnimStateNode_0 = Idle_pose() {
    is=title AnimStateNode
    stateEntered {guid=#1, notifyName=Entered_RandomLocoBreak}
    stateFullyBlended {guid=#2}
    stateLeft {guid=#3}
   }
   AnimStateNode_12 = Into_Combat_Pose() {
    is=title AnimStateNode
    stateEntered {guid=#4}
    stateFullyBlended {guid=#5}
    stateLeft {guid=#6}
   }
   AnimStateNode_15 = Elimination() {
    is=title AnimStateNode
    stateEntered {guid=#7}
    stateFullyBlended {guid=#8}
    stateLeft {guid=#9}
   }
   AnimStateNode_2 = Exit Combat Pose() {
    is=title AnimStateNode
    stateEntered {guid=#10}
    stateFullyBlended {guid=#11}
    stateLeft {guid=#12}
   }
   AnimStateNode_5 = Combat_pose() {
    is=title AnimStateNode
    stateEntered {guid=#13}
    stateFullyBlended {guid=#14}
    stateLeft {guid=#15}
   }
   AnimStateTransitionNode_1 = AnimStateTransitionNode()
   AnimStateTransitionNode_10 = AnimStateTransitionNode()
   AnimStateTransitionNode_16 = AnimStateTransitionNode()
   AnimStateTransitionNode_21 = AnimStateTransitionNode() {bAutomaticRuleBasedOnSequencePlayerInState=true}
   AnimStateTransitionNode_24 = AnimStateTransitionNode()
   AnimStateTransitionNode_27 = AnimStateTransitionNode() {bAutomaticRuleBasedOnSequencePlayerInState=true}
   AnimStateTransitionNode_34 = AnimStateTransitionNode() {
    transitionEnd {guid=#16}
    transitionInterrupt {guid=#17}
    transitionStart {guid=#18}
   }
   AnimStateTransitionNode_36 = AnimStateTransitionNode()
   AnimStateTransitionNode_4 = AnimStateTransitionNode() {
    bAutomaticRuleBasedOnSequencePlayerInState=true
    transitionEnd {guid=#19}
    transitionInterrupt {guid=#20}
    transitionStart {guid=#21}
   }
  ]
  links [
   {
    dst=Elimination
    rule=AnimGraph.AnimGraphNode_StateMachine_2.Backpack States.AnimStateTransitionNode_34.Transition
    src=Combat_pose
    via=AnimStateTransitionNode_34
   }
   {
    dst=Exit Combat Pose
    rule=AnimGraph.AnimGraphNode_StateMachine_2.Backpack States.AnimStateTransitionNode_10.Transition
    src=Combat_pose
    via=AnimStateTransitionNode_10
   }
   {
    dst=Idle_pose
    rule=AnimGraph.AnimGraphNode_StateMachine_2.Backpack States.AnimStateTransitionNode_27.Transition
    src=Elimination
    via=AnimStateTransitionNode_27
   }
   {
    dst=Combat_pose
    rule=AnimGraph.AnimGraphNode_StateMachine_2.Backpack States.AnimStateTransitionNode_1.Transition
    src=Exit Combat Pose
    via=AnimStateTransitionNode_1
   }
   {
    dst=Elimination
    rule=AnimGraph.AnimGraphNode_StateMachine_2.Backpack States.AnimStateTransitionNode_36.Transition
    src=Exit Combat Pose
    via=AnimStateTransitionNode_36
   }
   {
    dst=Idle_pose
    rule=AnimGraph.AnimGraphNode_StateMachine_2.Backpack States.AnimStateTransitionNode_21.Transition
    src=Exit Combat Pose
    via=AnimStateTransitionNode_21
   }
   {
    dst=Into_Combat_Pose
    rule=AnimGraph.AnimGraphNode_StateMachine_2.Backpack States.AnimStateTransitionNode_16.Transition
    src=Idle_pose
    via=AnimStateTransitionNode_16
   }
   {
    dst=Combat_pose
    rule=AnimGraph.AnimGraphNode_StateMachine_2.Backpack States.AnimStateTransitionNode_4.Transition
    src=Into_Combat_Pose
    via=AnimStateTransitionNode_4
   }
   {
    dst=Elimination
    rule=AnimGraph.AnimGraphNode_StateMachine_2.Backpack States.AnimStateTransitionNode_24.Transition
    src=Into_Combat_Pose
    via=AnimStateTransitionNode_24
   }
  ]
 }
 Idle_pose {
  path=AnimGraph.AnimGraphNode_StateMachine_2.Backpack States.AnimStateNode_0.Idle_pose
  nodes [Output Animation Pose(Result=Sequence Player 'OysterKnock_IdlePose_pack'() {bLoopAnimation=false, sequence=&OysterKnock_IdlePose_pack})]
 }
 Into_Combat_Pose {
  path=AnimGraph.AnimGraphNode_StateMachine_2.Backpack States.AnimStateNode_12.Into_Combat_Pose
  nodes [Output Animation Pose(Result=Sequence Player 'OysterKnock_Combat_into_pack'() {bLoopAnimation=false, sequence=/BRCosmetics/Accessories/FORT_Backpacks/Backpack_M_MED_OysterKnock/Anim/OysterKnock_Combat_into_pack})]
 }
 Elimination {
  path=AnimGraph.AnimGraphNode_StateMachine_2.Backpack States.AnimStateNode_15.Elimination
  nodes [Output Animation Pose(Result=Sequence Player 'OysterKnock_Elim_pack'() {bLoopAnimation=false, sequence=/BRCosmetics/Accessories/FORT_Backpacks/Backpack_M_MED_OysterKnock/Anim/OysterKnock_Elim_pack})]
 }
 Exit Combat Pose {
  path=AnimGraph.AnimGraphNode_StateMachine_2.Backpack States.AnimStateNode_2.Exit Combat Pose
  nodes [Output Animation Pose(Result=Sequence Player 'OysterKnock_Combat_out_pack'() {bLoopAnimation=false, sequence=/BRCosmetics/Accessories/FORT_Backpacks/Backpack_M_MED_OysterKnock/Anim/OysterKnock_Combat_out_pack})]
 }
 Combat_pose {
  path=AnimGraph.AnimGraphNode_StateMachine_2.Backpack States.AnimStateNode_5.Combat_pose
  nodes [Output Animation Pose(Result=Sequence Player 'OysterKnock_CombatPose_pack'() {bLoopAnimation=false, sequence=/BRCosmetics/Accessories/FORT_Backpacks/Backpack_M_MED_OysterKnock/Anim/OysterKnock_CombatPose_pack})]
 }
 Transition {
  path=AnimGraph.AnimGraphNode_StateMachine_2.Backpack States.AnimStateTransitionNode_1.Transition
  nodes [Result(bCanEnterTransition=Get isCombat() -> (isCombat:bool))]
 }
 Transition {
  path=AnimGraph.AnimGraphNode_StateMachine_2.Backpack States.AnimStateTransitionNode_10.Transition
  nodes [Result(bCanEnterTransition=NOT Boolean(A=Get isCombat()))]
 }
 Transition {
  path=AnimGraph.AnimGraphNode_StateMachine_2.Backpack States.AnimStateTransitionNode_16.Transition
  nodes [Result(bCanEnterTransition=Get isCombat())]
 }
 Transition {
  path=AnimGraph.AnimGraphNode_StateMachine_2.Backpack States.AnimStateTransitionNode_21.Transition
  nodes [Result(bCanEnterTransition=NOT Boolean(A=Get isCombat()))]
 }
 Transition {
  path=AnimGraph.AnimGraphNode_StateMachine_2.Backpack States.AnimStateTransitionNode_24.Transition
  nodes [Result(bCanEnterTransition=Get bEnemyPlayerEliminated())]
 }
 Transition {path=AnimGraph.AnimGraphNode_StateMachine_2.Backpack States.AnimStateTransitionNode_27.Transition, nodes [Result()]}
 Transition {
  path=AnimGraph.AnimGraphNode_StateMachine_2.Backpack States.AnimStateTransitionNode_34.Transition
  nodes [Result(bCanEnterTransition=Get bEnemyPlayerEliminated())]
 }
 Transition {
  path=AnimGraph.AnimGraphNode_StateMachine_2.Backpack States.AnimStateTransitionNode_36.Transition
  nodes [Result(bCanEnterTransition=Get bEnemyPlayerEliminated())]
 }
 Transition {path=AnimGraph.AnimGraphNode_StateMachine_2.Backpack States.AnimStateTransitionNode_4.Transition, nodes [Result()]}
 EventGraph {nodes [
   Delay(Duration=2)
   > Set Spinning Rotator B(Spinning Rotator B:Rotator="0,0,-180") -> (Output_Get:Rotator)
   Event BlueprintUpdateAnimation()
   > n1 = Sequence!(execute) -> (then_0, then_1, then_2:Exec)
   n2 = GetOwnerSkeletalMeshComponent()
   n3 = BreakRotator(InRot=BreakTransform(InTransform=GetDeltaTransformFromRefPose(self=@n2, BoneName=spine_05, BaseName=spine_01)).Rotation)
   > Set backpack_rotate(backpack_rotate=MakeRotator(Roll=MapRangeClamped(Value=@n3.Yaw, InRangeA=0, InRangeB=-30, OutRangeB=20)))
   > Set backpack_translation(backpack_translation=MakeVector(Y=MapRangeClamped(Value=@n3.Yaw, InRangeA=0, InRangeB=-20, OutRangeB=-4)))
   n4 = Get bIsJumping()
   n5 = Get bIsFalling()
   > Set JumpingFalling(JumpingFalling:float="To Float (Boolean)"(InBool=OR Boolean(A=@n4, B=@n5))) -> (Output_Get:float)
   n6 = Gravity Override State Function DEPRECATED(
    execute=@n1.then_1
    CharacterPartAnimInstance=Self-Reference()
    GameTimeInSeconds=Get Game Time in Seconds() -> (Game Time in Seconds:double)
    GravityOverrideParams=Get Gravity Override Params()
   )
   > Set Gravity Override(Gravity Override=@n6)
   Sequence!(execute=@n1.then_2) -> (then_0, then_1)
   n7 = Get bRecentlyFired() -> (bRecentlyFired:bool)
   n8 = Get bIsPlayingMeleeAnim()
   n9 = OR Boolean(A=OR Boolean(A=@n8, B=@n7), B=AND Boolean(A=@n8, B=@n7))
   n10 = OR Boolean(A=OR Boolean(A=Get bIsTargeting(), B=false), B=false)
   n11 = OR Boolean(A=AND Boolean(A=Get isCombat(), B=false), B=false)
   > Branch(Condition=OR Boolean(A=OR Boolean(A=@n10, B=@n9), B=OR Boolean(A=@n11, B=false)))
   > Set isCombat(isCombat:bool=true) -> (Output_Get:bool)
   > RetriggerableDelay(Duration=1.3)
   > Set isCombat(isCombat=false)
   Event BlueprintPostEvaluateAnimation() {is=title K2Node_Event}
   > Set bEnemyPlayerEliminated(bEnemyPlayerEliminated:bool=false) -> (Output_Get:bool)
   AnimNotify_ReactiveReset() {is=title K2Node_Event}
   > n12 = Set bEnemyPlayerEliminated(bEnemyPlayerEliminated=false)
   Branch(execute=[@n13.then, @n12.then], Condition=Get bShouldDisplayFrontendReactivity())
   > Delay(Duration=10)
   > Set bEnemyPlayerEliminated(bEnemyPlayerEliminated=true)
   n13 = Event BlueprintInitializeAnimation()
   n14 = Event BlueprintProcessAnimInputEvent()
   > n15 = Sequence()
   > Branch(Condition=OR Boolean(
     A=IsNamed(self=@n14.AnimInputEvent, InEventName=aie.PlayerVitality.HealthDecreased)
     B=IsNamed(self=@n14.AnimInputEvent, InEventName=Noneaie.PlayerVitality.ShieldsDecreased)
    ))
   > Set TookDamage(TookDamage:bool=true) -> (Output_Get:bool)
   > RetriggerableDelay(Duration=2)
   > Set TookDamage(TookDamage=false)
   Branch(execute=@n15.then_1, Condition=IsNamed(self=@n14.AnimInputEvent, InEventName=aie.EnemyPlayerEliminated))
   > Set bEnemyPlayerEliminated(bEnemyPlayerEliminated=true)
   > RetriggerableDelay(Duration=1.3)
   > Set bEnemyPlayerEliminated(bEnemyPlayerEliminated=false)
   > Set isCombat(isCombat=false)
   Reroute Node()
   > Set isEngaged(isEngaged:bool=OR Boolean(A=Get bIsTargeting(), B=Get bIsPlayingMeleeAnim())) -> (Output_Get:bool)
   > n16 = Branch(Condition=Get isEngaged() -> (isEngaged:bool))
   > Set isEngaged(isEngaged=true)
   Set isEngaged(execute=@n16.else, isEngaged=false)
   Reroute Node()
   n17 = Get bIsTargeting()
   > Set isCombat(isCombat=AND Boolean(A=Get bRecentlyFired(), B=NOT Boolean(A=@n17)))
   > n18 = Branch(Condition=Get isCombat())
   > Set isCombat(isCombat=true)
   Set isCombat(execute=@n18.else, isCombat=false)
   TryGetPawnOwner()
   Set Spinning Rotator A(Spinning Rotator A:Rotator="0,0,-180") -> (Output_Get:Rotator)
   Set Spinning Rotator A(Spinning Rotator A="0,0,180")
   Set Spinning Rotator B(Spinning Rotator B="0,0,180")
  ]}
]
variables [
 backpack_rotate {type=Rotator}
 backpack_translation {type=Vector}
 bEnemyPlayerEliminated {type=boolean}
 Game Time in Seconds {type=number}
 Gravity Override {type=Vector}
 Gravity Override Params {type=GravityOverrideParamsStruct}
 isCombat {type=boolean}
 isEngaged {type=boolean}
 IsInFE {type=number}
 JumpingFalling {type=number}
 Spinning Rotator A {type=Rotator}
 Spinning Rotator B {type=Rotator}
 TookDamage {type=boolean}
]
